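; Volcano construction-site actor that dresses the scene by build phase, with floodlights and a welder robot.
; On the authority, the phase comes from a progressive calendar POI actor at BeginPlay or from SetPhase.
; OnRep_ConstructionPhase shows spotlights in later phases and respawns BP_Welderbot at a transform picked by phase plus random offset.
; The construction script attaches FX meshes and lights to four spotlights and optionally previews every bot placement.
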(event EventBeginPlay
  (Networking|SwitchHasAuthority
    (:Authority
      (Variables|Default|SetRandom (Math|Random|RandomIntegerinRange -3))
      (bind _outactors (Actor|GetAllActorsOfClass "/Figment_S09/Athena/Prototype/Blueprints/CalendarProps/BP_CalendarDynamicPOIProgressive.BP_CalendarDynamicPOIProgressive_C"))
      (Utilities|IsValid (Utilities|Array|Get(acopy) _outactors)
        (:Is%20Valid
          (Variables|Default|SetConstructionPhase (Variables|/Figment_S09/Athena/Prototype/Blueprints/CalendarProps/BP_CalendarDynamicPOIProgressive.BP_CalendarDynamicPOIProgressive_C|GetProgressStage (Utilities|Array|Get(acopy) _outactors))))))))

(event Custom|SetPhase
  (params (Phase :type Integer))
  (Variables|Default|SetRandom (Math|Random|RandomIntegerinRange -3))
  (Variables|Default|SetConstructionPhase Phase))

(fn OnRep_ConstructionPhase
  (Utilities|FlowControl|Sequence
    (:then_0
      (for _array_element (Utilities|Array|MakeArray (Variables|Default|GetSpotlight1) (Variables|Default|GetSpotlight2) (Variables|Default|GetSpotlight3) (Variables|Default|GetSpotlight4))
        (Rendering|SetVisibility _array_element (>= (Variables|Default|GetConstructionPhase) 3) true))
      (Transformation|SetWorldLocation (Variables|Default|GetSpotlightPosition) (Variables|Default|GetPlatformWorldLocation) false false))
    (:then_1
      (Networking|SwitchHasAuthority
        (:Authority
          (bind _welderbot (Variables|Default|GetWelderbot)
            (:then
              (Actor|DestroyActor _welderbot))
            (:else
              _))
          (if (> (Variables|Default|GetConstructionPhase) 0)
            (bind _returnvalue (Game|SpawnActorfromClass "/Figment_S09/Athena/Environments/Volcano/Meshes/Welderbot/BP_Welderbot.BP_Welderbot_C" (Math|Transform|ComposeTransforms (Utilities|Array|Get(acopy) (Variables|Default|GetWelderbotTransforms) (Math|Integer|Clamp(Integer) (+ (* (Variables|Default|GetConstructionPhase) 3) (Variables|Default|GetRandom)) 0 (- (Utilities|Array|Length (Variables|Default|GetWelderbotTransforms)) 1))) (Transformation|GetActorTransform)) "AlwaysSpawn" "OverrideRootScale" 0 0))
            (Variables|Default|SetWelderbot _returnvalue)))))))

(fn ConstructionScript
  (Utilities|FlowControl|Sequence
    (:then_0
      (for _array_element (Utilities|Array|MakeArray (Variables|Default|GetSpotlight1) (Variables|Default|GetSpotlight2) (Variables|Default|GetSpotlight3) (Variables|Default|GetSpotlight4))
        (bind _returnvalue (AddComponent|NODE_AddFXStaticMeshComponent-0 :bManualAttachment true :RelativeTransform (Transformation|GetRelativeTransform _array_element)))
        (bind _returnvalue_2 (AddComponent|NODE_AddFXStaticMeshComponent-1 :bManualAttachment true :RelativeTransform (Transformation|GetRelativeTransform _array_element)))
        (bind _returnvalue_3 (AddComponent|NODE_AddSpotLightComponent-0 :bManualAttachment true :RelativeTransform (Transformation|GetRelativeTransform _array_element)))
        (Transformation|AttachComponentToComponent _returnvalue _array_element "None" "SnapToTarget" "SnapToTarget" "KeepRelative" true)
        (Transformation|AttachComponentToComponent _returnvalue_2 _array_element "None" "SnapToTarget" "SnapToTarget" "KeepRelative" true)
        (Transformation|AttachComponentToComponent _returnvalue_3 _array_element "None" "SnapToTarget" "SnapToTarget" "KeepRelative" true))
      (Transformation|SetWorldLocation (Variables|Default|GetSpotlightPosition) (Variables|Default|GetPlatformWorldLocation) false false))
    (:then_1
      (if (Variables|Default|GetPreviewBots)
        (for _array_element_2 (Variables|Default|GetWelderbotTransforms)
          (bind _returnvalue_4 (AddComponent|NODE_AddStaticMeshComponent-0 :RelativeTransform _array_element_2))
          (Rendering|Material|SetScalarParameterValueonMaterials _returnvalue_4 "ArmMove" 1.0))))))
