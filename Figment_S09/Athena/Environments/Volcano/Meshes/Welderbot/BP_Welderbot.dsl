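; Ambient Welderbot prop for the Volcano environment that periodically welds and plays vocal cues.
; Two looping timers, with intervals randomized at BeginPlay, trigger welding and speech; welding is skipped while its duration timer is active.
; The StartWeld timeline animates the arm through the ArmMove material parameter, starts sparks and audio on forward completion, then reverses after a random duration.
; The blink light's "Flare" particle spawn event triggers the blink sound.
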(event EventBeginPlay
  (Utilities|Time|SetTimerbyEvent (EventDispatchers|CreateEvent :!event Custom|WeldABit) (Math|Random|RandomFloatinRange 3.0 6.0) true)
  (Utilities|Time|SetTimerbyEvent (EventDispatchers|CreateEvent :!event Custom|SpeakABit) (Math|Random|RandomFloatinRange 6.0 12.0) true))

(event Custom|WeldABit
  (if (not (Utilities|Time|IsTimerActivebyHandle (Variables|Default|GetWeldDurationTimer)))
    (Audio|PlaySoundatLocation (Variables|Default|GetArmMoveSound) (Transformation|GetWorldLocation (Variables|Default|GetSM_Welderbot)))
    (bind (_direction _armmove) (Timeline|StartWeld Play)
      (:Update
        (Rendering|Material|SetScalarParameterValueonMaterials (Variables|Default|GetSM_Welderbot) "ArmMove" _armmove))
      (:Finished
        (if (Utilities|Enum|Equal(Enum) _direction "Forward")
          (Components|Activation|SetActive (Variables|Default|GetP_Doggus_Weld_Sparks) true true)
          (bind _returnvalue (Utilities|Time|SetTimerbyEvent (EventDispatchers|CreateEvent :!event Custom|StopWelding) (Math|Random|RandomFloatinRange 0.5 3.0)))
          (Variables|Default|SetWeldDurationTimer _returnvalue)
          (Audio|Components|Audio|FadeIn (Variables|Default|GetWeldingAudio) 1.0 1.0 (Math|Random|RandomFloatinRange 0.0 2.9) "Linear" :!class /Script/Engine.AudioComponent)
          (Audio|PlaySoundatLocation (Variables|Default|GetWeldStartSound) (Transformation|GetWorldLocation (Variables|Default|GetWeldingAudio))))))))

(event Custom|StopWelding
  (bind _sm_welderbot (Variables|Default|GetSM_Welderbot))
  (bind _returnvalue (Transformation|GetWorldLocation _sm_welderbot))
  (Components|Activation|SetActive (Variables|Default|GetP_Doggus_Weld_Sparks) false false)
  (Audio|Components|Audio|FadeOut (Variables|Default|GetWeldingAudio) 1.0 0.0 "Linear" :!class /Script/Engine.AudioComponent)
  (Audio|PlaySoundatLocation (Variables|Default|GetArmMoveSound) _returnvalue)
  (Audio|PlaySoundatLocation (Variables|Default|GetWeldStopSound) _returnvalue)
  (Timeline|StartWeld Reverse))

(event Custom|SpeakABit
  (Audio|PlaySoundatLocation "/Figment_S09/Athena/Sounds/SupplyDrops/Drone/WelderBot/WelderBot_Vocal_Cue.WelderBot_Vocal_Cue" (Transformation|GetWorldLocation (Variables|Default|GetAmbientAudio))))

(event OnParticleSpawn(P_Welderbot_Light_Blink)
  (params (EventName :type Name :default "None") (EmitterTime :type Float%20%28single-precision%29) (Location :type Vector :default "0, 0, 0") (Velocity :type Vector :default "0, 0, 0"))
  (if (Utilities|Name|Equal(Name) EventName "Flare")
    (Audio|PlaySoundatLocation (Variables|Default|GetLightBlinkSound) (Transformation|GetWorldLocation (Variables|Default|GetWeldingAudio)))))

(fn ConstructionScript)
