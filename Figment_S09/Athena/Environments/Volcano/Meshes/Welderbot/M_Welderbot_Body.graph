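# Default Lit body material for the Welderbot in the Volcano environment, reusing Supply Drop Drone textures with a hue-shifted diffuse and tinted emissive.
# An ArmMove parameter drives vertex animation around a local Y axis.
# Green vertex color masks rotation about Pivot1, while blue masks the translation of Pivot2 around it, so that part follows without rotating.
# A shared WPO bob function is added to the arm motion.
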Material /Figment_S09/Athena/Environments/Volcano/Meshes/Welderbot/M_Welderbot_Body
graph {
 properties {physMaterial=/Game/FortressPhysicalMaterials/Metal, pixelDepthOffsetMode=PDOM_Legacy, usedShadingModels=Default Lit}
 nodes [
  n1 = VertexColor()
  n2 = Transform(None=Constant3Vector() {constant=<rgba 0 1 0 0>}) {transformSourceType=TRANSFORMSOURCE_Local}
  n3 = LinearInterpolate(Alpha=ScalarParameter() {parameterName=ArmMove}) {constA=-0.1, constB=0.35}
  n4 = Transform(None=VectorParameter() {parameterName=Pivot1, defaultValue=<rgba 0 0 -79.286 0>}.RGB) {transformSourceType=TRANSFORMSOURCE_Local}
  n5 = RotateAboutAxis(NormalizedRotationAxis=@n2, RotationAngle=@n3, PivotPoint=@n4, Position=Subtract(A=WorldPosition().XYZ, B=ActorPositionWS()))
  n6 = RotateAboutAxis(
   NormalizedRotationAxis=@n2
   RotationAngle=@n3
   PivotPoint=@n4
   Position=Transform(None=VectorParameter() {parameterName=Pivot2, defaultValue=<rgba 60 0 -120 0>}.RGB) {transformSourceType=TRANSFORMSOURCE_Local}
  )
  Add(
   A=Add(A=Multiply(A=@n1.G, B=@n5), B=Multiply(A=@n1.B, B=@n6))
   B=MaterialFunctionCall() {materialFunction=/Game/Athena/Environments/Volcano/Meshes/Welderbot/MF_Welderbot_WPOBob}.Result
  )
  MaterialFunctionCall(
   Hue Shift Percentage=ScalarParameter() {defaultValue=0.05, parameterName=HueShift}
   Texture=TextureSampleParameter2D() {parameterName=Diffuse, texture=/Game/Athena/Items/Consumables/SuperDingo/Textures/T_Supply_Drop_Drone_D}.RGB
  ) {materialFunction=&HueShift}
  Multiply(
   A=TextureSampleParameter2D() {parameterName=Emissive, texture=/Game/Athena/Items/Consumables/SuperDingo/Textures/T_Supply_Drop_Drone_E}.R
   B=VectorParameter() {parameterName=Emissive, defaultValue=<rgba 0.9069 3.9942 5 0>}.RGB
  )
  TextureSampleParameter2D() {parameterName=Masks, texture=/Game/Athena/Items/Consumables/SuperDingo/Textures/T_Supply_Drop_Drone_S}
  TextureSampleParameter2D() {parameterName=Normal, samplerType=SAMPLERTYPE_Normal, texture=/Game/Athena/Items/Consumables/SuperDingo/Textures/T_Supply_Drop_Drone_N}
 ]
}
