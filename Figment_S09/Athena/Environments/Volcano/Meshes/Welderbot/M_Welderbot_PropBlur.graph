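# Translucent, two-sided unlit material that fakes a motion-blurred spinning propeller on the volcano Welderbot, with a bob function for world position offset.
# A panning water-blur mask breaks up a U-gradient and sphere-mask fade and distorts a fan-blade gradient texture.
# That gradient is tinted by fake lighting built from time-of-day color, sun/moon reflection highlights and a shadow tint.
# On ES2 and ES3_1, a flat constant replaces the lighting.
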Material /Figment_S09/Athena/Environments/Volcano/Meshes/Welderbot/M_Welderbot_PropBlur
graph {
 properties {blendMode=BLEND_Translucent, pixelDepthOffsetMode=PDOM_Legacy, shadingModel=MSM_Unlit, twoSided=true, usedShadingModels=Unlit}
 nodes [
  MaterialFunctionCall() {materialFunction=/Game/Athena/Environments/Volcano/Meshes/Welderbot/MF_Welderbot_WPOBob}
  n1 = ComponentMask(None=TextureCoordinate()) {r=true}
  n2 = OneMinus(None=@n1)
  n3 = SphereMask(A=@n1, B=Constant() {r=0.5}) {attenuationRadius=0.25, hardnessPercent=0}
  n4 = TextureCoordinate()
  n5 = AppendVector(A=ScalarParameter() {defaultValue=3, parameterName=X}, B=ScalarParameter() {defaultValue=1, parameterName=Y})
  n6 = Time()
  n7 = ScalarParameter() {defaultValue=8, parameterName=PanSpeed}
  n8 = TextureSample(UVs=Panner(Coordinate=Multiply(A=@n4, B=@n5), Time=Multiply(A=@n6, B=@n7)) {speedX=0.1, speedY=-1}) {
   samplerType=SAMPLERTYPE_LinearGrayscale
   texture=/Game/Characters/Player/Male/Medium/Heads/M_MED_AUX_Alien_Head_01/Textures/T_waterBlur__MASK
  }
  Multiply(A=Multiply(A=Add(A=@n2, B=@n3), B=Add(A=@n2, B=@n8.G))) {constB=1.25}
  n9 = Add(A=MaterialFunctionCall() {materialFunction=&Color_Time_of_Day}.Result) {constB=0.25}
  n10 = CollectionParameter() {
   collection=&FortniteMaterialParameters
   parameterName=SunAndMoonModelDirectionalVector
  }
  n11 = DotProduct(A=ComponentMask(None=@n10) {b=true, g=true, r=true}, B=ReflectionVectorWS() {bNormalizeCustomWorldNormal=true})
  n12 = Power(Base=@n11, Exp=ScalarParameter() {defaultValue=8, parameterName=Spec Highlight Power}) {constExponent=20}
  n13 = Add(A=MaterialFunctionCall() {materialFunction=&Color_Time_of_Day}.Result) {constB=0.25}
  n14 = Multiply(A=Saturate(None=Power(Base=@n11)), B=@n13)
  n15 = VectorParameter() {parameterName=ShadowTint, defaultValue=<rgba 0.3255 0.3691 0.5 0>}
  n16 = Add(
   A=Multiply(A=Multiply(A=@n9) {constB=0.8}, B=Saturate(None=@n12))
   B=Add(A=Multiply(A=@n14) {constB=0.5}, B=Multiply(A=@n13, B=@n15.RGB) {constB=0.05})
  )
  n17 = Constant() {r=0.7}
  n18 = Constant2Vector() {r=0.02}
  n19 = TextureSample(UVs=Add(A=TextureCoordinate(), B=Multiply(A=@n18, B=@n8.G))) {texture=/Game/Athena/Environments/Blueprints/SkyTube/FX/T_Fan_Blade_Gradient}
  Multiply(
   A=FeatureLevelSwitch(Default=@n16, ES2=@n17, ES3_1=@n17, SM4_REMOVED=@n16)
   B=Multiply(A=Multiply(A=@n19.RGB) {constB=1.5}, B=Add(A=@n8.G) {constB=0.7})
  )
 ]
}
